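; Camera modifier for the IcicleLarva punch that limits how fast the view can turn, judging by its name while punching.
; It overrides BlueprintProcessViewRotation to clamp the pitch and yaw of the incoming delta rotation to symmetric limits.
; Each limit is a max angle variable scaled by DeltaTime, making the cap a frame-rate independent turn rate.
; Roll delta is zeroed, the view rotation passes through unchanged, and the function returns false.
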(fn BlueprintProcessViewRotation
  (params (ViewTarget :type Actor%20Object%20Reference) (DeltaTime :type Float%20%28single-precision%29) (InViewRotation :type Rotator :default "0, 0, 0") (InDeltaRot :type Rotator :default "0, 0, 0"))
  (return InViewRotation (Math|Rotator|MakeRotator 0.0 (Math|Float|ClampAngle (.pitch InDeltaRot) (* (* DeltaTime (Variables|Default|GetMaxPitchAngle)) -1.0) (* DeltaTime (Variables|Default|GetMaxPitchAngle))) (Math|Float|ClampAngle (.yaw InDeltaRot) (* (* DeltaTime (Variables|Default|GetMaxYawAngle)) -1.0) (* DeltaTime (Variables|Default|GetMaxYawAngle)))) false))
